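# Physics asset for the skeletal mesh of the Speedy glider, a VinderTech GliderChute gadget among the Battle Royale cosmetics.
# Its constraint list is empty, so it defines no joints between physics bodies.
# No body shapes or other non-default settings appear, so it holds no visible ragdoll or secondary-motion setup.
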PhysicsAsset /BRCosmetics/Gadgets/Assets/VinderTech_GliderChute/Glider_Speedy/Meshes/SK_Glider_Speedy_PhysicsAsset
constraints []
